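# Sound cue for the ready sound of the PajamaParty cosmetic pickaxe in Athena.
# A random node picks among seven equally weighted pajama pickaxe swing waves, then a modulator adds slight volume variation and a raised, varied pitch.
# It plays quietly in the melee weapon sound class with close-range projectile attenuation and a pickaxe-ready property sheet.
# Overridden concurrency prevents new instances at the limit, and the cue primes on load.
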SoundCue /BRCosmetics/Sounds/PickAxes/PajamaParty/PickAxe_Pajama_Ready_Athena_Cue
root=Modulator_0
nodes [
 {id=Modulator_0, children [Random_5], properties {pitchMax=1.2, pitchMin=1.1, volumeMax=1.1, volumeMin=1}}
 {
  id=Random_5
  children [WavePlayer_5, WavePlayer_21, WavePlayer_22, WavePlayer_23, WavePlayer_24, WavePlayer_25, WavePlayer_26]
  properties {weights [1,1,1,1,1,1,1]}
 }
 {id=WavePlayer_21, properties {soundWaveAssetPtr=&/PickAxe_Pajama_Swing02}}
 {id=WavePlayer_22, properties {soundWaveAssetPtr=&/PickAxe_Pajama_Swing03}}
 {id=WavePlayer_23, properties {soundWaveAssetPtr=&/PickAxe_Pajama_Swing04}}
 {id=WavePlayer_24, properties {soundWaveAssetPtr=&/PickAxe_Pajama_Swing05}}
 {id=WavePlayer_25, properties {soundWaveAssetPtr=&/PickAxe_Pajama_Swing06}}
 {id=WavePlayer_26, properties {soundWaveAssetPtr=&/PickAxe_Pajama_Swing07}}
 {id=WavePlayer_5, properties {soundWaveAssetPtr=&/PickAxe_Pajama_Swing01}}
]
properties {
 attenuationSettings=&Fort_Weapon_Projectile_Close_Athena_Att
 audioPropertiesSheet=/BRCosmetics/Sounds/Settings/PropertySheets/Pickaxe/APS_Cosmetic_Pickaxe_Ready
 bOverrideConcurrency=true
 bPrimeOnLoad=true
 maxDistance=4950
 soundClassObject=&Fort_WeaponMelee
 volumeMultiplier=0.2
 concurrencyOverrides {resolutionRule=PreventNew}
}
